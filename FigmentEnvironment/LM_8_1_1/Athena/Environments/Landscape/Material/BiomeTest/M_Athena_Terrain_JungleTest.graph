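# Test terrain material for a jungle biome in the Athena landscape environment, recoloring the standard grass layer.
# It uses Material Attributes, taking the full layer from the MF_Athena_Grass_01 function.
# MatLayerBlend_MultiplyBaseColor tints that layer's base color with the Grass_JungleColor parameter, a green-cyan by default, with the Mask input left unconnected.
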Material /FigmentEnvironment/LM_8_1_1/Athena/Environments/Landscape/Material/BiomeTest/M_Athena_Terrain_JungleTest
graph {
 properties {bUseMaterialAttributes=true, pixelDepthOffsetMode=PDOM_Legacy, usedShadingModels=Default Lit}
 nodes [MaterialFunctionCall!(
   Material=MaterialFunctionCall() {materialFunction=/FigmentEnvironment/LM_8_1_1/Athena/Environments/Landscape/MaterialFunctions/Standard/MF_Athena_Grass_01}.Result
   NewBaseColor=VectorParameter() {parameterName=Grass_JungleColor, defaultValue=<rgba 0.141 1 0.6731 0>}.RGB
   Mask
  ) {materialFunction=/Engine/Functions/MaterialLayerFunctions/MatLayerBlend_MultiplyBaseColor}]
}
