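# Material function for mountain landscape materials that tints grass and rock diffuse inputs with colors from the MPC_Landscape parameter collection.
# Grass and Rock outputs multiply each diffuse by MountainGrassTint or MountainRockTint, and the rock tint is also output directly.
# Two blended color outputs lerp rock tint toward constant 1 using separate alpha inputs.
# GrassTint outputs that constant rather than the collection tint, and two local vector parameters sit unconnected.
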MaterialFunction /FigmentEnvironment/LM_8_1_1/Athena/Environments/Landscape/MaterialFunctions/MF_MountainGrass_Colors
graph {nodes [
  CollectionParameter() {collection=/FigmentEnvironment/LM_8_1_1/Athena/Environments/Landscape/MPC/MPC_Landscape, parameterName=MountainGrassTint}
  n1 = CollectionParameter() {collection=&MPC_Landscape, parameterName=MountainGrassTint}
  FunctionOutput(None=Multiply(A=FunctionInput() {inputName=Grass_Diffuse}, B=ComponentMask(None=@n1) {b=true, g=true, r=true})) {outputName=Grass, sortPriority=2}
  n2 = ComponentMask(None=CollectionParameter() {collection=&MPC_Landscape, parameterName=MountainRockTint}) {b=true, g=true, r=true}
  n3 = Constant() {r=1}
  FunctionOutput(None=LinearInterpolate(A=@n2, B=@n3, Alpha=FunctionInput() {bUsePreviewValueAsDefault=true, inputName=ColorBlend_Alpha, inputType=FunctionInput_Scalar})) {outputName=Blended_Colors, sortPriority=2}
  FunctionOutput(None=@n2) {outputName=RockTint, sortPriority=1}
  FunctionOutput(None=Multiply(A=FunctionInput() {inputName=Rock_Diffuse}, B=@n2)) {outputName=Rock}
  FunctionOutput(None=@n3) {outputName=GrassTint, sortPriority=3}
  FunctionOutput(None=LinearInterpolate(A=@n2, B=@n3, Alpha=FunctionInput() {bUsePreviewValueAsDefault=true, inputName=ColorBlend_Alpha_Bias, inputType=FunctionInput_Scalar})) {outputName=Blended_Colors_Bias, sortPriority=2}
  VectorParameter() {parameterName=MountainGrass_Tint, defaultValue=<rgba 0.3913 0.4533 1.0287 0>}
  VectorParameter() {parameterName=MountainRock_Tint, defaultValue=<rgba 0.5 0.47 0.445 0>}
 ]}
