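# Biome-test master landscape material for Athena terrain, blending painted biome and ground layers and driving landscape grass.
# Slope-mixed, dual-scale world-aligned rock with height-gradient shading feeds HeightLerps for four grass biomes, with a parallel rock/grass normal blend.
# Forest, mud, road, gravel and farm field are height-blended over these before a soft-biome-mask topology adjustment.
# Eight grass outputs subtract combinations of rock, layer and noise masks to limit foliage placement.
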Material /FigmentEnvironment/LM_8_1_1/Athena/Environments/Landscape/Material/BiomeTest/M_Athena_Terrain_Master_Biome
graph {
 properties {pixelDepthOffsetMode=PDOM_Legacy, usedShadingModels=Default Lit}
 nodes [
  Constant() {r=0.9}
  n1 = LandscapeLayerSample() {parameterName=Grass_Topology, previewWeight=1}
  n2 = ScalarParameter() {defaultValue=1534.9661, parameterName=TexSize}
  n3 = MaterialFunctionCall!(
   TextureObject=TextureObject() {texture=/FigmentEnvironment/LM_8_1_1/Packages/DS_Fortnite_Terrain_NoLOD/Textures/T_Cave_Exterior_Wall_D}
   TextureSize=@n2
   WorldPosition
   Export Float 4=StaticBool() {value=true}
   World Space Normal_Z=Multiply(A=VertexNormalWS(), B=ScalarParameter() {defaultValue=0.725, parameterName=RockProjNormalMP_Z})
   TextureObject_Z=TextureObject() {texture=/FigmentEnvironment/LM_8_1_1/Packages/DS_Fortnight/WhiteRockCustom_D}
   World Space Normal=Multiply(A=VertexNormalWS(), B=ScalarParameter() {defaultValue=1, parameterName=RockProjNormalMP})
   ProjectionTransitionContrast=ScalarParameter() {defaultValue=10, parameterName=RockProjContrast}
  ) {materialFunction=/FigmentEnvironment/LM_8_1_1/Athena/Prototype/Terrain/WorldAlignedTexture_ZTex}
  n4 = ScalarParameter() {defaultValue=3740.7383, parameterName=TexSize_2}
  n5 = MaterialFunctionCall(
   TextureObject=TextureObject() {texture=&T_Cave_Exterior_Wall_D}
   TextureSize=@n4
   Export Float 4=StaticBool() {value=true}
   World Space Normal_Z=Multiply(A=VertexNormalWS(), B=ScalarParameter() {defaultValue=0.75, parameterName=RockProjNormalMP_Z_1})
   TextureObject_Z=TextureObject() {texture=&WhiteRockCustom_D}
   World Space Normal=Multiply(A=VertexNormalWS(), B=ScalarParameter() {defaultValue=1, parameterName=RockProjNormalMP_1})
   ProjectionTransitionContrast=ScalarParameter() {defaultValue=10, parameterName=RockProjContrast_2}
  ) {materialFunction=&WorldAlignedTexture_ZTex}
  n6 = ScalarParameter() {defaultValue=5, parameterName=TransitonRockWall_LOW}
  n7 = ScalarParameter() {defaultValue=-5, parameterName=TransitonRockWall_High}
  n8 = ComponentMask(None=VertexNormalWS()) {b=true}
  n9 = Clamp(None=Add(A=LinearInterpolate(A=@n6, B=@n7, Alpha=@n8), B=ScalarParameter() {defaultValue=0.578, parameterName=TransitonRockWall_Offset}))
  n10 = Multiply(
   A=ComponentMask(None=LinearInterpolate(A=@n3.XYZ Texture, B=@n5.XYZ Texture, Alpha=@n9)) {b=true, g=true, r=true}
   B=MaterialFunctionCall!(height_Offset, height_GradScale, Coords=WorldPosition().XYZ) {materialFunction=/FigmentEnvironment/LM_8_1_1/Athena/Environments/Landscape/MaterialFunctions/MF_Athena_WorldHeightGrad}.Result
  )
  n11 = LandscapeLayerCoords() {mappingScale=4}
  n12 = TextureSampleParameter2D(UVs=@n11) {
   parameterName=Grass
   samplerSource=SSM_Wrap_WorldGroupSettings
   texture=/FigmentEnvironment/LM_8_1_1/Packages/DS_Fortnite_Terrain_NoLOD/Textures/T_Grasslands_D
  }
  n13 = Multiply(
   A=Multiply(A=@n12.A, B=ComponentMask(None=@n3.XYZ Texture) {a=true})
   B=ScalarParameter() {defaultValue=0.08, parameterName=Transition_Phase}
  ) {constB=1.1}
  n14 = ScalarParameter() {defaultValue=5, parameterName=Contrast}
  n15 = Multiply(A=ComponentMask(None=VertexNormalWS()) {b=true}, B=ScalarParameter() {defaultValue=2, parameterName=RockMultiply})
  MaterialExpressionMaterialFunctionCall_9 = MaterialFunctionCall(
   A=Multiply(A=@n10, B=VectorParameter() {parameterName=RockTint, defaultValue=<rgba 0.9 0.9 0.9 0>}.RGB)
   B=@n12.RGB
   Transition Phase=@n13
   Height Texture=@n15
   Contrast=@n14
  ) {materialFunction=&HeightLerp}
  n16 = OneMinus(None=Power(Base=@MaterialExpressionMaterialFunctionCall_9.Alpha) {constExponent=5})
  n17 = LandscapeLayerSample() {parameterName=Gravel}
  n18 = LandscapeLayerSample() {parameterName=Forest}
  n19 = Subtract(A=Subtract(A=Subtract(A=@n1, B=@n16), B=@n17), B=@n18)
  n20 = LandscapeLayerSample() {parameterName=Mud}
  n21 = LandscapeLayerSample() {parameterName=Road}
  n22 = LandscapeLayerSample() {parameterName=FarmField}
  MaterialExpressionSubtract_41 = Subtract(A=Subtract(A=Subtract(A=@n19, B=@n20), B=@n21), B=@n22)
  n23 = LandscapeLayerSample() {parameterName=Swamp_Grass, previewWeight=1}
  n24 = Subtract(A=Subtract(A=Subtract(A=@n23, B=@n16), B=@n17), B=@n18)
  MaterialExpressionSubtract_42 = Subtract(A=Subtract(A=Subtract(A=@n24, B=@n20), B=@n21), B=@n22)
  n25 = LandscapeLayerSample() {parameterName=Farm_Grass}
  n26 = Subtract(A=Subtract(A=Subtract(A=@n25, B=@n16), B=@n17), B=@n18)
  MaterialExpressionSubtract_43 = Subtract(A=Subtract(A=Subtract(A=@n26, B=@n20), B=@n21), B=@n22)
  n27 = LandscapeLayerSample() {parameterName=Mountain_Grass}
  n28 = Subtract(A=Subtract(A=Subtract(A=@n27, B=@n16), B=@n17), B=@n18)
  MaterialExpressionSubtract_44 = Subtract(A=Subtract(A=Subtract(A=@n28, B=@n20), B=@n21), B=@n22)
  n29 = TextureSample(UVs=LandscapeLayerCoords() {mappingScale=130}) {samplerType=SAMPLERTYPE_LinearColor, texture=/FigmentEnvironment/LM_8_1_1/Athena/Prototype/SoftBiomeTest/T_Athena_Terrain_SoftBiome_Mask}
  MaterialExpressionSubtract_20 = Subtract(A=Subtract(A=Subtract(A=@n1, B=@n29.R), B=@n16), B=@n21)
  n30 = Subtract(A=LandscapeLayerSample() {parameterName=Forest}, B=@n16)
  MaterialExpressionSubtract_45 = Subtract(A=Subtract(A=Subtract(A=@n30, B=@n20), B=@n21), B=@n22)
  MaterialExpressionReroute_24 = Reroute(None=Subtract(A=LandscapeLayerSample() {parameterName=Gravel}, B=@n16))
  n31 = TextureSample(UVs=LandscapeLayerCoords() {mappingScale=5}) {texture=/FigmentEnvironment/LM_8_1_1/Packages/DS_Fortnight/T_FORT_GardenSoil}
  n32 = Desaturation(None=@n31.RGB)
  n33 = Multiply(A=@n32) {constB=2}
  MaterialExpressionSubtract_26 = Subtract(A=LandscapeLayerSample() {parameterName=FarmField}, B=OneMinus(None=LinearInterpolate(Alpha=@n33) {constA=-1, constB=10}))
  LandscapeGrassOutput!(
   Grass=@MaterialExpressionSubtract_41
   Swamp_Grass=@MaterialExpressionSubtract_42
   Farm_Grass=@MaterialExpressionSubtract_43
   Mountain_Grass=@MaterialExpressionSubtract_44
   Flowers=@MaterialExpressionSubtract_20
   Forest=@MaterialExpressionSubtract_45
   Gravel=@MaterialExpressionReroute_24
   FarmField=@MaterialExpressionSubtract_26
  ) {grassTypes [
    Grass {
     grassType=/FigmentEnvironment/LM_8_1_1/Athena/Environments/Landscape/Foliage/Athena_Foliage_Grass
     input {
      expression=&self:MaterialExpressionSubtract_41
      inputName=None
      mask=0
      maskA=0
      maskB=0
      maskG=0
      maskR=0
      outputIndex=0
     }
    }
    Swamp_Grass {
     grassType=/FigmentEnvironment/LM_8_1_1/Athena/Environments/Landscape/Foliage/Athena_Foliage_Swamp_Grass
     input {
      expression=&self:MaterialExpressionSubtract_42
      inputName=None
      mask=0
      maskA=0
      maskB=0
      maskG=0
      maskR=0
      outputIndex=0
     }
    }
    Farm_Grass {
     grassType=/FigmentEnvironment/LM_8_1_1/Athena/Environments/Landscape/Foliage/Athena_Foliage_Farm_Grass
     input {
      expression=&self:MaterialExpressionSubtract_43
      inputName=None
      mask=0
      maskA=0
      maskB=0
      maskG=0
      maskR=0
      outputIndex=0
     }
    }
    Mountain_Grass {
     grassType=/FigmentEnvironment/LM_8_1_1/Athena/Environments/Landscape/Foliage/Athena_Foliage_Mountain_Grass
     input {
      expression=&self:MaterialExpressionSubtract_44
      inputName=None
      mask=0
      maskA=0
      maskB=0
      maskG=0
      maskR=0
      outputIndex=0
     }
    }
    Flowers {
     grassType=/FigmentEnvironment/LM_8_1_1/Athena/Environments/Landscape/Foliage/Athena_Foliage_Flowers
     input {
      expression=&self:MaterialExpressionSubtract_20
      inputName=None
      mask=0
      maskA=0
      maskB=0
      maskG=0
      maskR=0
      outputIndex=0
     }
    }
    Forest {
     grassType=/FigmentEnvironment/LM_8_1_1/Athena/Environments/Landscape/Foliage/Athena_Foliage_Forest
     input {
      expression=&self:MaterialExpressionSubtract_45
      inputName=None
      mask=0
      maskA=0
      maskB=0
      maskG=0
      maskR=0
      outputIndex=0
     }
    }
    Gravel {
     grassType=/FigmentEnvironment/LM_8_1_1/Athena/Environments/Landscape/Foliage/Athena_Foliage_Gravel
     input {
      expression=&self:MaterialExpressionReroute_24
      inputName=None
      mask=0
      maskA=0
      maskB=0
      maskG=0
      maskR=0
      outputIndex=0
     }
    }
    FarmField {
     grassType=/FigmentEnvironment/LM_8_1_1/Athena/Environments/Landscape/Foliage/Athena_Foliage_Farm_Field
     input {
      expression=&self:MaterialExpressionSubtract_26
      inputName=None
      mask=0
      maskA=0
      maskB=0
      maskG=0
      maskR=0
      outputIndex=0
     }
    }
   ]}
  MaterialExpressionLinearInterpolate_5 = LinearInterpolate(
   B=ScalarParameter() {defaultValue=0.3, group=Grass, parameterName=GrassSpecAmount}
   Alpha=Power(Base=ComponentMask(None=@n12.A) {a=true}, Exp=ScalarParameter() {defaultValue=2, group=Grass, parameterName=GrassSpecPower})
  )
  LandscapeLayerBlend!(Layer Grass_Topology=@MaterialExpressionLinearInterpolate_5) {layers [{
     blendType=LB_WeightBlend
     constHeightInput=0
     layerName=Grass_Topology
     previewWeight=0
     constLayerInput=<0 0 0>
     heightInput {expression=null, inputName=None, mask=0, maskA=0, maskB=0, maskG=0, maskR=0, outputIndex=0}
     layerInput {
      expression=&self:MaterialExpressionLinearInterpolate_5
      inputName=None
      mask=0
      maskA=0
      maskB=0
      maskG=0
      maskR=0
      outputIndex=0
     }
    }]}
  LandscapeLayerCoords()
  n34 = TextureObject() {samplerType=SAMPLERTYPE_LinearColor, texture=&T_Athena_Terrain_SoftBiome_Mask}
  n35 = TextureSampleParameter2D(UVs=LandscapeLayerCoords() {mappingScale=4}) {
   parameterName=Forest
   samplerSource=SSM_Wrap_WorldGroupSettings
   texture=/FigmentEnvironment/LM_8_1_1/Packages/DS_Fortnight/SM/Textures/T_FORT_ForestFloor01
  }
  n36 = TextureSampleParameter2D(UVs=LandscapeLayerCoords() {mappingScale=4}) {
   parameterName=Gravel
   samplerSource=SSM_Wrap_WorldGroupSettings
   texture=&WhiteRockCustom_D
  }
  n37 = Multiply(A=@n36.RGB, B=VectorParameter() {parameterName=Gravel_Tint, defaultValue=<rgba 0.635 0.6136 0.5461 0>}.RGB)
  n38 = MaterialFunctionCall(
   A=@n35.RGB
   B=@n37
   Transition Phase=ScalarParameter() {defaultValue=0.3, parameterName=GravelForest_Trans}
   Height Texture=@n36.A
   Contrast=ScalarParameter() {defaultValue=0.5, parameterName=GravelForest_Contrast}
  ) {materialFunction=&HeightLerp}
  n39 = Multiply(A=LandscapeLayerSample() {parameterName=Forest}, B=ScalarParameter() {defaultValue=1, parameterName=Forest_Transition})
  n40 = Multiply(A=@n36.A, B=@n35.A)
  n41 = ScalarParameter() {defaultValue=1, parameterName=Forest_Contrast}
  n42 = TextureSampleParameter2D(UVs=LandscapeLayerCoords() {mappingScale=4}) {
   parameterName=Swamp
   samplerSource=SSM_Wrap_WorldGroupSettings
   texture=/FigmentEnvironment/LM_8_1_1/Athena/Environments/Landscape/Textures/T_Athena_Swamp_Grass_01
  }
  n43 = MaterialFunctionCall!(Rock_Diffuse=@n10, Grass_Diffuse=@n42.RGB) {materialFunction=/FigmentEnvironment/LM_8_1_1/Athena/Environments/Landscape/MaterialFunctions/MF_SwampGrass_Colors}
  MaterialExpressionMaterialFunctionCall_5 = MaterialFunctionCall(
   A=@n43.Rock
   B=@n43.Rock
   Transition Phase=Multiply(A=Power(Base=@n42.A) {constExponent=0.5}, B=ScalarParameter() {defaultValue=0.0408, parameterName=Swamp_RockTransPhase}) {constB=1.1}
   Height Texture=@n15
   Contrast=ScalarParameter() {defaultValue=5, parameterName=Swamp_RockContrast}
  ) {materialFunction=&HeightLerp}
  n44 = TextureSampleParameter2D(UVs=LandscapeLayerCoords() {mappingScale=4}) {
   parameterName=Farm
   samplerSource=SSM_Wrap_WorldGroupSettings
   texture=/FigmentEnvironment/LM_8_1_1/Athena/Environments/Landscape/Textures/T_Athena_FarmGrass_01
  }
  n45 = MaterialFunctionCall!(Rock_Diffuse=@n10, Alpha=@n44.A, ColorBlend_Alpha, ColorBlend_Alpha_Bias, In, Grass_Diffuse=@n44.A) {materialFunction=/FigmentEnvironment/LM_8_1_1/Athena/Environments/Landscape/MaterialFunctions/MF_FarmGrass_Colors}
  n46 = TextureSample(UVs=LandscapeLayerCoords() {mappingScale=505}) {samplerType=SAMPLERTYPE_LinearColor, texture=&T_Athena_Terrain_SoftBiome_Mask}
  MaterialExpressionMaterialFunctionCall_17 = MaterialFunctionCall(
   A=@n45.Rock
   B=LinearInterpolate(A=@n45.Grass, B=@n12.RGB, Alpha=Power(Base=@n46.G) {constExponent=5})
   Transition Phase=Multiply(A=Power(Base=@n44.A) {constExponent=0.5}, B=ScalarParameter() {defaultValue=0.0408, parameterName=Farm_RockTransPhase}) {constB=1.1}
   Height Texture=@n15
   Contrast=ScalarParameter() {defaultValue=5, parameterName=Farm_RockContrast}
  ) {materialFunction=&HeightLerp}
  n47 = TextureSampleParameter2D(UVs=LandscapeLayerCoords() {mappingScale=4}) {
   parameterName=Mountain
   samplerSource=SSM_Wrap_WorldGroupSettings
   texture=/FigmentEnvironment/LM_8_1_1/Athena/Environments/Landscape/Textures/T_Athena_Grass_Messy_01
  }
  n48 = MaterialFunctionCall!(Rock_Diffuse=@n10, ColorBlend_Alpha, ColorBlend_Alpha_Bias, Grass_Diffuse=@n47.RGB) {materialFunction=/FigmentEnvironment/LM_8_1_1/Athena/Environments/Landscape/MaterialFunctions/MF_MountainGrass_Colors}
  MaterialExpressionMaterialFunctionCall_20 = MaterialFunctionCall(
   A=@n48.Rock
   B=@n48.Rock
   Transition Phase=Multiply(A=Power(Base=@n12.A) {constExponent=0.5}, B=ScalarParameter() {defaultValue=0.0408, parameterName=Mountain_RockTransPhase}) {constB=1.1}
   Height Texture=@n15
   Contrast=ScalarParameter() {defaultValue=5, parameterName=Mountain_RockContrast_1}
  ) {materialFunction=&HeightLerp}
  n49 = LandscapeLayerBlend!(
   Layer Grass_Topology=@MaterialExpressionMaterialFunctionCall_9.Results
   Layer Swamp_Grass=@MaterialExpressionMaterialFunctionCall_5.Results
   Layer Farm_Grass=@MaterialExpressionMaterialFunctionCall_17.Results
   Layer Mountain_Grass=@MaterialExpressionMaterialFunctionCall_20.Results
  ) {layers [
    {
     blendType=LB_WeightBlend
     constHeightInput=0
     layerName=Grass_Topology
     previewWeight=1
     constLayerInput=<0 0 0>
     heightInput {expression=null, inputName=None, mask=0, maskA=0, maskB=0, maskG=0, maskR=0, outputIndex=0}
     layerInput {
      expression=&self:MaterialExpressionMaterialFunctionCall_9
      inputName=None
      mask=0
      maskA=0
      maskB=0
      maskG=0
      maskR=0
      outputIndex=0
     }
    }
    {
     blendType=LB_WeightBlend
     constHeightInput=0
     layerName=Swamp_Grass
     previewWeight=1
     constLayerInput=<0 0 0>
     heightInput {expression=null, inputName=None, mask=0, maskA=0, maskB=0, maskG=0, maskR=0, outputIndex=0}
     layerInput {
      expression=&self:MaterialExpressionMaterialFunctionCall_5
      inputName=None
      mask=0
      maskA=0
      maskB=0
      maskG=0
      maskR=0
      outputIndex=0
     }
    }
    {
     blendType=LB_WeightBlend
     constHeightInput=0.5
     layerName=Farm_Grass
     previewWeight=1
     constLayerInput=<0 0 0>
     heightInput {expression=null, inputName=None, mask=0, maskA=0, maskB=0, maskG=0, maskR=0, outputIndex=0}
     layerInput {
      expression=&self:MaterialExpressionMaterialFunctionCall_17
      inputName=None
      mask=0
      maskA=0
      maskB=0
      maskG=0
      maskR=0
      outputIndex=0
     }
    }
    {
     blendType=LB_WeightBlend
     constHeightInput=0
     layerName=Mountain_Grass
     previewWeight=1
     constLayerInput=<0 0 0>
     heightInput {expression=null, inputName=None, mask=0, maskA=0, maskB=0, maskG=0, maskR=0, outputIndex=0}
     layerInput {
      expression=&self:MaterialExpressionMaterialFunctionCall_20
      inputName=None
      mask=0
      maskA=0
      maskB=0
      maskG=0
      maskR=0
      outputIndex=0
     }
    }
   ]}
  n50 = LandscapeLayerCoords() {mappingScale=4}
  n51 = Multiply(
   A=TextureSampleParameter2D(UVs=@n50) {parameterName=Mud, texture=/FigmentEnvironment/LM_8_1_1/Packages/DS_Fortnight/Mud}.RGB
   B=VectorParameter() {parameterName=Mud_Tint, defaultValue=<rgba 0.9262 0.9432 0.95 0>}.RGB
  )
  n52 = LandscapeLayerSample() {parameterName=Mud}
  n53 = ScalarParameter() {defaultValue=1, parameterName=Mud_Transition}
  n54 = OneMinus(None=@n12.A)
  n55 = OneMinus(None=@n36.A)
  n56 = TextureSample(UVs=@n50) {texture=/FigmentEnvironment/LM_8_1_1/Packages/DS_Fortnight/SM/Textures/T_Street_Road_D}
  n57 = LandscapeLayerSample() {parameterName=Road}
  n58 = Power(Base=@n57)
  n59 = ScalarParameter() {defaultValue=1, parameterName=RoadMud_Transition}
  n60 = MaterialFunctionCall(
   A=MaterialFunctionCall(
    A=MaterialFunctionCall(A=@n49, B=@n38.Results, Transition Phase=@n39, Height Texture=@n40, Contrast=@n41) {materialFunction=&HeightLerp}.Results
    B=@n51
    Transition Phase=Multiply(A=@n52, B=@n53)
    Height Texture=Multiply(A=@n54, B=@n55)
    Contrast=ScalarParameter() {defaultValue=0.5, parameterName=Mud_Contrast}
   ) {materialFunction=&HeightLerp}.Results
   B=MaterialFunctionCall(
    A=@n51
    B=@n56.RGB
    Transition Phase=Multiply(A=@n58, B=@n59)
    Height Texture=@n56.RGB
    Contrast=ScalarParameter() {defaultValue=5, parameterName=Road_Contrast}
   ) {materialFunction=&HeightLerp}.Results
   Transition Phase=Multiply(A=@n57, B=ScalarParameter() {defaultValue=1.2, parameterName=Raod_Transition})
   Height Texture=Multiply(A=OneMinus(None=@n12.A), B=ScalarParameter() {defaultValue=0.25, parameterName=Road_heightMP})
   Contrast=ScalarParameter() {defaultValue=5, parameterName=RoadMud_Contrast}
  ) {materialFunction=&HeightLerp}
  n61 = LandscapeLayerSample() {parameterName=Gravel}
  n62 = ScalarParameter() {defaultValue=1, parameterName=Gravel_BlendContrast}
  n63 = VectorParameter() {parameterName=FarmField_Tint, defaultValue=<rgba 1 1 1 0>}
  MaterialFunctionCall!(
   Mask 2=@n34
   UV Coordinates=LandscapeLayerCoords() {mappingScale=505}
   RockMask
   BaseTexture=MaterialFunctionCall(
    A=MaterialFunctionCall(A=@n60.Results, B=@n37, Transition Phase=@n61, Height Texture=@n36.A, Contrast=@n62) {materialFunction=&HeightLerp}.Results
    B=Multiply(A=@n31.RGB, B=@n63.RGB)
    Transition Phase=LandscapeLayerSample() {parameterName=FarmField}
    Height Texture=@n32
    Contrast=ScalarParameter() {defaultValue=0.5, parameterName=FarmField_Contrast}
   ) {materialFunction=&HeightLerp}.Results
  ) {materialFunction=/FigmentEnvironment/LM_8_1_1/Athena/Environments/Landscape/MaterialFunctions/MF_TerrainTopoAdjustment}
  Power(Base=@n44.B)
  n64 = TextureSample(UVs=@n11) {
   samplerSource=SSM_Wrap_WorldGroupSettings
   samplerType=SAMPLERTYPE_Normal
   texture=/FigmentEnvironment/LM_8_1_1/Packages/DS_Fortnite_Terrain_NoLOD/Textures/T_Grasslands_N
  }
  n65 = TextureObject() {samplerType=SAMPLERTYPE_Normal, texture=/FigmentEnvironment/LM_8_1_1/Packages/DS_Fortnite_Terrain_NoLOD/Textures/T_Cave_Exterior_Wall_N}
  n66 = MaterialFunctionCall!(
   TextureObject=@n65
   TextureSize=@n2
   WorldSpace
   Normal
   ProjectionTransitionContrast=ScalarParameter() {defaultValue=-10, parameterName=RockProjContrast_Normal}
   WorldPosition
   Use High Quality Normals
  ) {materialFunction=/FigmentEnvironment/LM_8_1_1/Athena/Environments/Materials/WorldAlignedNormalContrast}
  n67 = TextureObject() {samplerType=SAMPLERTYPE_Normal, texture=&T_Cave_Exterior_Wall_N}
  n68 = MaterialFunctionCall(
   TextureObject=@n67
   TextureSize=@n4
   ProjectionTransitionContrast=ScalarParameter() {defaultValue=-10, parameterName=RockProjContrast_Normal_1}
  ) {materialFunction=&WorldAlignedNormalContrast}
  MaterialExpressionReroute_17 = Reroute(None=MaterialFunctionCall(A=LinearInterpolate(A=@n66.XYZFlatTop, B=@n68.XYZFlatTop, Alpha=@n9), B=@n64.RGB, Transition Phase=@n13, Height Texture=@n15, Contrast=@n14) {materialFunction=&HeightLerp}.Results)
  MaterialExpressionConstant3Vector_1 = Constant3Vector() {constant=<rgba 0.5 0.5 1 0>}
  n69 = LandscapeLayerBlend!(
   Layer Grass_Topology=@MaterialExpressionReroute_17
   Layer Swamp_Grass=@MaterialExpressionReroute_17
   Layer Farm_Grass=@MaterialExpressionReroute_17
   Layer Mountain_Grass=@MaterialExpressionReroute_17
   Layer Road=@MaterialExpressionConstant3Vector_1
  ) {layers [
    {
     blendType=LB_WeightBlend
     constHeightInput=0
     layerName=Grass_Topology
     previewWeight=0
     constLayerInput=<0 0 0>
     heightInput {expression=null, inputName=None, mask=0, maskA=0, maskB=0, maskG=0, maskR=0, outputIndex=0}
     layerInput {
      expression=&self:MaterialExpressionReroute_17
      inputName=None
      mask=0
      maskA=0
      maskB=0
      maskG=0
      maskR=0
      outputIndex=0
     }
    }
    {
     blendType=LB_WeightBlend
     constHeightInput=0
     layerName=Swamp_Grass
     previewWeight=0
     constLayerInput=<0 0 0>
     heightInput {expression=null, inputName=None, mask=0, maskA=0, maskB=0, maskG=0, maskR=0, outputIndex=0}
     layerInput {
      expression=&self:MaterialExpressionReroute_17
      inputName=None
      mask=0
      maskA=0
      maskB=0
      maskG=0
      maskR=0
      outputIndex=0
     }
    }
    {
     blendType=LB_WeightBlend
     constHeightInput=0
     layerName=Farm_Grass
     previewWeight=0
     constLayerInput=<0 0 0>
     heightInput {expression=null, inputName=None, mask=0, maskA=0, maskB=0, maskG=0, maskR=0, outputIndex=0}
     layerInput {
      expression=&self:MaterialExpressionReroute_17
      inputName=None
      mask=0
      maskA=0
      maskB=0
      maskG=0
      maskR=0
      outputIndex=0
     }
    }
    {
     blendType=LB_WeightBlend
     constHeightInput=0
     layerName=Mountain_Grass
     previewWeight=0
     constLayerInput=<0 0 0>
     heightInput {expression=null, inputName=None, mask=0, maskA=0, maskB=0, maskG=0, maskR=0, outputIndex=0}
     layerInput {
      expression=&self:MaterialExpressionReroute_17
      inputName=None
      mask=0
      maskA=0
      maskB=0
      maskG=0
      maskR=0
      outputIndex=0
     }
    }
    {
     blendType=LB_WeightBlend
     constHeightInput=0
     layerName=Road
     previewWeight=0
     constLayerInput=<0 0 0>
     heightInput {expression=null, inputName=None, mask=0, maskA=0, maskB=0, maskG=0, maskR=0, outputIndex=0}
     layerInput {
      expression=&self:MaterialExpressionConstant3Vector_1
      inputName=None
      mask=0
      maskA=0
      maskB=0
      maskG=0
      maskR=0
      outputIndex=0
     }
    }
   ]}
  Reroute(None=@n69)
  ScalarParameter() {defaultValue=306200, parameterName=TerrainSize}
  Subtract(A=@MaterialExpressionLinearInterpolate_5)
 ]
}
